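# Niagara system for the GroundCracks vehicle drift trail cosmetic, drawing effects at wheel sockets on the car's skeletal mesh.
# Glow sprites, wheel-wrap meshes and ribbon trails accompany GPU sparks and rock debris, with particles partitioned by wheel and masked through FN_DisableTireFX.
# Painted-variant modules set each emitter's colors, some with R1 overrides.
# Speed, frontend-preview and game-mode inputs adjust spark rate and motion and rock scale.
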NiagaraSystem /VehicleCosmetics/DriftTrail/DriftTrail_GroundCracks/FX/NS_DriftTrail_GroundCracks_MP
dependencies {
 usedDataInterfaces [
  &NiagaraDataInterfaceArrayColor
  &NiagaraDataInterfaceCurve
  &NiagaraDataInterfaceColorCurve
  &NiagaraDataInterfaceArrayFloat3
  &NiagaraDataInterfaceSkeletalMesh
 ]
 usedDynamicInputs []
 usedModules [
  &self:SystemSpawnScript.SystemScriptSource.SystemScriptGraph.NiagaraNodeAssignment_0.SetVariables_#1
  &FN_UnpackVehicleFlags
  /VehicleCosmetics/FX/Niagara/Modules/VC_SetVehicleSpeedAttribs
  &SystemState
  /Game/Effects/Niagara/Modules/Vehicle/FN_VehicleFlags_Cheatcodes
 ]
}
emitters {
 Glow_Fake001 {
  emitterUpdateScript [
   &EmitterState()
   &SpawnBurst_Instantaneous(Spawn Count=2, Use Spawn Probability=false, UseLoopCountLimit=false)
   &SpawnRate(
    SpawnRate=/VehicleCosmetics/FX/Niagara/Modules/VC_Mask_Floats(
     FrontEnd Value:NiagaraFloat=800
     blsFrontEndPreview:NiagaraBool=$User.bIsFrontendPreview
     SpeedMin Value:NiagaraFloat=400
     SpeedMax Value:NiagaraFloat=180
     Speed Alpha:NiagaraFloat=$System.Speed_Alpha
    )
    Use Spawn Probability=false
   ) {enabled=false}
  ]
  particleSpawnScript [
   &VC_PaintedVariantsOverride(
    Black=<rgba 0 0 0 1>
    BurntSienna=<rgba 0.37 0.1575 0 3>
    Cobalt=<rgba 0 0.2554 1 3>
    Crimson=<rgba 1 0.1038 0 1>
    DefaultColor=<rgba 1 0.5756 0 1>
    ForestGreen=<rgba 0.289 1 0 1>
    Grey=<rgba 0.4583 0.4583 0.4583 1>
    Lime=<rgba 0.6577 1 0 1>
    Orange=<rgba 1 0.3323 0 2>
    Pink=<rgba 1 0 0.6845 2>
    Purple=<rgba 0.4599 0 1 2>
    Saffron=<rgba 1 0.9866 0 2>
    SkyBlue=<rgba 0 0.7482 1 2>
    TitaniumWhite=<rgba 1 1 1 1>
   )
   &InitializeParticle(
    Color Mode=::NewEnumerator0
    Lifetime=$Emitter.CurrentLoopDuration
    Sprite Rotation Mode=::NewEnumerator2
    Sprite Size Mode=::NewEnumerator3
    Uniform Sprite Size=200
    UsePositionOffset=false
    Write Material Random=true
   )
   &PartitionParticles()
   &SkeletalMeshLocation(
    Bone / Socket Index=&Add_Integer(A=$Particles.PartitionParticles.Partition, B=4)
    Bone / Socket Sampling Mode=::NewEnumerator5
    Interpolate Mesh Sampling=false
    MaskSpawnGroup=false
    Mesh Sampling Type=::NewEnumerator6
    Sampled Position Offset=&AddVectorToPosition(Position=<-59.002 0 -20>, Vector=$User.offset)
    Skeletal Mesh=$System.SkeletalMesh
   )
   &AddVelocity(Velocity=&VC_Mask_Vectors(
     FrontEnd Value:Vector3f=&RandomRangeVector(Minimum=<-600 150 500>, Maximum=<-500 -150 100>)
     blsFrontEndPreview:NiagaraBool=$User.bIsFrontendPreview
     SpeedMin Value:Vector3f=&RandomRangeVector(Minimum=<-300 50 500>, Maximum=<-500 -50 100>)
     SpeedMax Value:Vector3f=<0 0 0>
     Speed Alpha:NiagaraFloat=$System.Speed_Alpha
    )) {enabled=false}
  ]
  particleUpdateScript [
   &FN_DisableTireFX(
    Mask Index Attribute:NiagaraInt32=$Particles.PartitionParticles.Partition
    Mask Index FL=2
    Mask Index FR=3
    Mask Index RL=0
    Mask Index RR=1
   )
   &ParticleState()
   &SolveForcesAndVelocity(Clamp Velocity=false, Limit Acceleration=false)
  ]
  rendererClasses [&NiagaraSpriteRendererProperties]
  renderers [{rendererClass=&NiagaraSpriteRendererProperties}]
 }
 Rocks_V001 {
  simTarget=GPUComputeSim
  emitterUpdateScript [
   &EmitterState()
   &SpawnRate(
    SpawnRate=&VC_Mask_Floats(
     FrontEnd Value:NiagaraFloat=500
     blsFrontEndPreview:NiagaraBool=$User.bIsFrontendPreview
     SpeedMin Value:NiagaraFloat=500
     SpeedMax Value:NiagaraFloat=500
     Speed Alpha:NiagaraFloat=$System.Speed_Alpha
    )
    Use Spawn Probability=false
   )
  ]
  particleSpawnScript [
   &VC_PaintedVariantsOverride(
    Black=<rgba 0.018 0.0158 0.0112 1>
    BurntSienna=<rgba 0.025 0.0161 0.0106 1>
    Cobalt=<rgba 0.0168 0.0169 0.025 1>
    Crimson=<rgba 0.028 0.0124 0.009 1>
    DefaultColor=&RandomRangeLinearColor(Minimum=<rgba 0.04 0.0375 0.0346 1>, Maximum=<rgba 0.03 0.0276 0.0231 1>)
    ForestGreen=<rgba 0.0152 0.025 0.0109 1>
    Grey=<rgba 0.02 0.0183 0.013 1>
    Lime=<rgba 0.0368 0.05 0.0237 1>
    Orange=<rgba 0.04 0.0236 0.0139 1>
    Pink=<rgba 0.045 0.0224 0.0215 1>
    Purple=<rgba 0.024 0.0149 0.018 1>
    R1 DefaultColor=&RandomRangeLinearColor(Minimum=<rgba 0.04 0.0375 0.0346 1>, Maximum=<rgba 0.03 0.0276 0.0231 1>)
    R1 Lime:LinearColor=<rgba 0.1105 0.15 0.0711 1>
    Saffron=<rgba 0.042 0.0342 0.0252 1>
    SkyBlue=<rgba 0.0086 0.022 0.0205 1>
    TitaniumWhite=<rgba 0.08 0.0676 0.0592 1>
    Use R1 Painted Overrides=true
   )
   &InitializeParticle(
    Color Mode=::NewEnumerator0
    Lifetime=0.27
    Mesh Scale Max=/VehicleCosmetics/FX/Niagara/Modules/GameMode/VC_SetVectorByGameModeID(
     SameValueForLockerAndBR?:NiagaraBool=true
     Locker Vector:Vector3f=&Multiply_VectorByFloat(Vector=<0.132 0.132 0.231>, Float=0.1)
     Battle Royale Vector:Vector3f=&Multiply_VectorByFloat(Vector=<0.132 0.132 0.231>, Float=0.1)
     Rocket Racing Vector:Vector3f=&Multiply_VectorByFloat(Vector=<0.132 0.132 0.231>, Float=0.2)
     GameModeID:NiagaraInt32=$System.GameModeID
    )
    Mesh Scale Min=&VC_SetVectorByGameModeID(
     SameValueForLockerAndBR?:NiagaraBool=true
     Locker Vector:Vector3f=&Multiply_VectorByFloat(Vector=<0.215 0.225 0.245>, Float=0.1)
     Battle Royale Vector:Vector3f=&Multiply_VectorByFloat(Vector=<0.215 0.225 0.245>, Float=0.1)
     Rocket Racing Vector:Vector3f=&Multiply_VectorByFloat(Vector=<0.215 0.225 0.245>, Float=0.1)
     GameModeID:NiagaraInt32=$System.GameModeID
    )
    Mesh Scale Mode=::NewEnumerator2
    UsePositionOffset=false
    Write Material Random=true
   )
   &PartitionParticles(Partition Mode:ENiagara_PartitionMode=::NewEnumerator1)
   &SkeletalMeshLocation(
    Bone / Socket Index=&Add_Integer(A=$Particles.PartitionParticles.Partition, B=4)
    Bone / Socket Sampling Mode=::NewEnumerator5
    Interpolate Mesh Sampling=false
    MaskSpawnGroup=false
    Mesh Sampling Type=::NewEnumerator6
    Sampled Position Offset=&SelectVectorFromArray(
     Direct Array Index=&ModuloInt(A=&ReturnExecIndex(), B=2)
     Vector Selection Array=DataInterface() {propertyValues {FloatData [<0 15 -25>, <0 -15 -25>]}}
    )
    Skeletal Mesh=$System.SkeletalMesh
   )
   &AddVelocity(Velocity=&VC_Mask_Vectors(
     FrontEnd Value:Vector3f=<-2000 0 0>
     blsFrontEndPreview:NiagaraBool=$User.bIsFrontendPreview
     SpeedMin Value:Vector3f=<-2000 0 0>
     SpeedMax Value:Vector3f=<-2000 0 0>
     Speed Alpha:NiagaraFloat=$System.Speed_Alpha
    ))
   &InitialMeshOrientation(
    Mesh Orientation Mode=::NewEnumerator7
    Rotation=&RandomRangeVector(
     Minimum=&SelectVectorFromArray(
      Direct Array Index=&ModuloInt(A=&ReturnExecIndex(), B=2)
      Vector Selection Array=DataInterface() {propertyValues {FloatData [<-0.19 0 0.15>, <0.19 0 0.15>]}}
     )
     Maximum=&SelectVectorFromArray(
      Direct Array Index=&ModuloInt(A=&ReturnExecIndex(), B=2)
      Vector Selection Array=DataInterface() {propertyValues {FloatData [<-0.1 0 -0.15>, <0.1 0 -0.15>]}}
     )
    )
    Use Rotation Vector=true
   )
  ]
  particleUpdateScript [
   &FN_DisableTireFX(
    Mask Index Attribute:NiagaraInt32=$Particles.PartitionParticles.Partition
    Mask Index FL=2
    Mask Index FR=3
    Mask Index RL=0
    Mask Index RR=1
   )
   &ParticleState()
   &InheritSourceMovement(
    Applied Position Scale=&Multiply_VectorByFloat(
     Vector=<1 1 1>
     Float=&FloatFromCurve(FloatCurve=DataInterface() {propertyValues {Curve {keys [
          {interpMode=RCIM_Cubic, leaveTangent=-0.0051, tangentMode=RCTM_Break, time=0, value=1}
          {arriveTangent=-0.247, interpMode=RCIM_Cubic, tangentMode=RCTM_Break, time=1, value=0.9}
         ]}}})
    )
    ApplyForce=false
    ApplyPosition=true
    ApplyPositionFalloff=false
   )
   &DynamicMaterialParameters(
    Index 0 Param 1=1
    Module.Index0Float0Write=true
    Module.Index0Float1Write=true
    Module.Index0Float2Write=true
    Module.Index0Float3Write=true
   )
   &ScaleMeshSize(Scale Factor=&Multiply_VectorByFloat(
     Vector=<1 1 1>
     Float=&FloatFromCurve(
      FloatCurve=DataInterface() {propertyValues {Curve {keys [
          {interpMode=RCIM_Cubic, tangentMode=RCTM_Break, time=0, value=1}
          {arriveTangent=-0.113, interpMode=RCIM_Cubic, tangentMode=RCTM_Break, time=0.2, value=1}
          <at 1 1>
         ]}}}
      Scale Curve=30
     )
    ))
   &SolveForcesAndVelocity(Clamp Velocity=false, Limit Acceleration=false)
  ]
  rendererClasses [&NiagaraMeshRendererProperties, &NiagaraSpriteRendererProperties]
  renderers [
   {rendererClass=&NiagaraMeshRendererProperties, rendererIndex=0}
   {rendererClass=&NiagaraSpriteRendererProperties, rendererIndex=1}
  ]
 }
 Sparks001 {
  simTarget=GPUComputeSim
  emitterUpdateScript [
   &EmitterState()
   &SpawnRate(
    SpawnRate=&VC_Mask_Floats(
     FrontEnd Value:NiagaraFloat=700
     blsFrontEndPreview:NiagaraBool=$User.bIsFrontendPreview
     SpeedMin Value:NiagaraFloat=400
     SpeedMax Value:NiagaraFloat=180
     Speed Alpha:NiagaraFloat=$System.Speed_Alpha
    )
    Use Spawn Probability=false
   )
  ]
  particleSpawnScript [
   &VC_PaintedVariantsOverride(
    Black=<rgba 0.005 0.005 0.005 10>
    BurntSienna=<rgba 20 5.085 0 3>
    Cobalt=<rgba 1 3.6854 20 3>
    Crimson=<rgba 30 1.8594 2.7548 5>
    DefaultColor=&SelectLinearColorFromArray(
     Array Sampling Mode=::NewEnumerator0
     Direct Array Index=?
     Color Selection Array=DataInterface() {propertyValues {ColorData [<rgba 60 34.831 4.0625 1>, <rgba 1 0.5847 0 1>, <rgba 1 1 1 0>]}}
     Random Seed=0
    )
    ForestGreen=<rgba 0.0375 0.25 0.0482 3>
    Grey=<rgba 0.07 0.07 0.07 1>
    Lime=<rgba 5.7031 10 1.4062 5>
    Orange=<rgba 40 17.5538 3.234 2>
    Pink=<rgba 100 40.625 81.8248 1>
    Purple=<rgba 0.6156 0 1.5 1>
    Saffron=<rgba 20 17.9078 3.75 5>
    SkyBlue=<rgba 1.276 3.7539 5 1>
    TitaniumWhite=<rgba 40 40 40 1>
   )
   &InitializeParticle(
    Color=$Particles.Color
    Lifetime=&MakeCustomFloatFromBool(
     Bool=$User.bIsFrontendPreview
     True Float=&RandomRangeFloat(Minimum=0.025, Maximum=0.3)
     False Float=&RandomRangeFloat(Minimum=0.1, Maximum=0.3)
    )
    Sprite Size=&RandomRangeVector2D(
     Minimum=&Multiply_Vector2D_ByFloat(Float=5, Vector2D=<0.8 9>)
     Maximum=&Multiply_Vector2D_ByFloat(Float=5, Vector2D=<0.3 3>)
    )
    Sprite Size Mode=::NewEnumerator4
    UsePositionOffset=false
    Write Material Random=true
   )
   &PartitionParticles()
   &SkeletalMeshLocation(
    Bone / Socket Index=&Add_Integer(A=$Particles.PartitionParticles.Partition, B=4)
    Bone / Socket Sampling Mode=::NewEnumerator5
    Interpolate Mesh Sampling=false
    MaskSpawnGroup=false
    Mesh Sampling Type=::NewEnumerator6
    Sampled Position Offset=&AddVectorToPosition(Position=<-40 0 -20>, Vector=$User.offset)
    Skeletal Mesh=$System.SkeletalMesh
   )
   &AddVelocity(Velocity=&VC_Mask_Vectors(
     FrontEnd Value:Vector3f=&RandomRangeVector(Minimum=<-600 150 500>, Maximum=<-500 -150 100>)
     blsFrontEndPreview:NiagaraBool=$User.bIsFrontendPreview
     SpeedMin Value:Vector3f=&RandomRangeVector(Minimum=<-800 50 500>, Maximum=<-1000 -50 100>)
     SpeedMax Value:Vector3f=<0 0 0>
     Speed Alpha:NiagaraFloat=$System.Speed_Alpha
    )) {enabled=false}
   &AddVelocity(Velocity=&VC_SetVectorByGameModeID(
     SameValueForLockerAndBR?:NiagaraBool=false
     Locker Vector:Vector3f=&RandomRangeVector(Minimum=<-600 150 500>, Maximum=<-500 -150 100>)
     Battle Royale Vector:Vector3f=&RandomRangeVector(Minimum=<-800 50 500>, Maximum=<-1000 -50 100>)
     Rocket Racing Vector:Vector3f=&RandomRangeVector(Minimum=<-800 50 500>, Maximum=<-100 -50 100>)
     GameModeID:NiagaraInt32=$System.GameModeID
    )) {module=AddVelocity001}
   &ShapeLocation(
    MaskSpawnGroup=false
    Override Seed=false
    Sphere Radius=&RandomRangeFloat(Minimum=20, Maximum=30)
    UseOwnerScale=true
   )
  ]
  particleUpdateScript [
   &FN_DisableTireFX(
    Mask Index Attribute:NiagaraInt32=$Particles.PartitionParticles.Partition
    Mask Index FL=2
    Mask Index FR=3
    Mask Index RL=0
    Mask Index RR=1
   )
   &ParticleState()
   &AccelerationForce(Acceleration=&RandomRangeVector(
     Minimum=&Multiply_VectorByFloat(Vector=<0 0 -1000>, Float=1.5)
     Maximum=&Multiply_VectorByFloat(Vector=<0 0 -1500>, Float=1.5)
    ))
   &SolveForcesAndVelocity(Clamp Velocity=false, Limit Acceleration=false)
   &ScaleColor(
    Scale Alpha=&FloatFromCurve(FloatCurve=DataInterface() {propertyValues {Curve {keys [{interpMode=RCIM_Cubic, tangentMode=RCTM_Break, time=0, value=1}, {interpMode=RCIM_Cubic, tangentMode=RCTM_Break, time=1, value=0}]}}})
    ScaleA=true
    ScaleRGB=true
   )
  ]
  rendererClasses [&NiagaraSpriteRendererProperties]
  renderers [{rendererClass=&NiagaraSpriteRendererProperties}]
 }
 Trail001 {
  emitterUpdateScript [
   &EmitterState()
   &VC_RateQualityHandler(High Spawn Rate=50, Low Spawn Rate=30, Medium Spawn Rate=40, Use Custom Rate?=true)
   &VC_SpawnRibbon_v1(Number of Ribbons=2, SpawnRate=$Emitter.Spawn Rate)
  ]
  particleSpawnScript [
   &VC_PaintedVariantsOverride(
    Black=<rgba 0.0311 0.0311 0.0311 10>
    BurntSienna=<rgba 0.3206 0.0608 0.0037 1>
    Cobalt=<rgba 0 0.8762 4.5 0.7>
    Crimson=<rgba 3 0 0.0813 1>
    DefaultColor=<rgba 1 1 1 0.2>
    ForestGreen=<rgba 0 0.5 0.0253 0.7>
    Grey=<rgba 0.18 0.18 0.18 1>
    Lime=<rgba 2.5 5 0 0.7>
    Orange=<rgba 12 3.76 0 0.7>
    Pink=<rgba 7 0 4.9215 0.6>
    Purple=<rgba 0.6667 0 2 0.7>
    R1 Black:LinearColor=<rgba 0 0 0 0.6>
    R1 BurntSienna:LinearColor=<rgba 0.8796 0.1365 0.072 1>
    R1 Cobalt:LinearColor=<rgba 0 0.2665 0.8 1>
    R1 Crimson:LinearColor=<rgba 0.5092 0 0.0085 1>
    R1 DefaultColor=<rgba 0.5583 0.314 0.15 0.8>
    R1 ForestGreen:LinearColor=<rgba 0.0694 0.6 0 0.6>
    R1 Grey:LinearColor=<rgba 0.25 0.25 0.25 1>
    R1 Lime:LinearColor=<rgba 0.2504 0.4 0 1>
    R1 Orange:LinearColor=<rgba 0.5 0.1037 0.0555 1>
    R1 Pink:LinearColor=<rgba 0.6257 0.2051 0.6724 1>
    R1 Purple:LinearColor=<rgba 0.1511 0 0.6724 1>
    R1 Saffron:LinearColor=<rgba 0.4485 0.4544 0.0454 1>
    R1 SkyBlue:LinearColor=<rgba 0 0.2662 0.4 1>
    R1 TitaniumWhite:LinearColor=<rgba 0.4 0.4 0.4 1>
    Saffron=<rgba 4.5448 5 0 0.7>
    SkyBlue=<rgba 0 1.9614 4 0.7>
    TitaniumWhite=<rgba 4 4 4 1>
    Use R1 Painted Overrides=true
   ) {enabled=false}
   &VC_InitializeRibbon_v1(
    Color=$Particles.Color
    Custom Offset=true
    Enable Skeletal Position=true
    Facing Axis:ENiagaraOrientationAxis=::YAxis
    Local Vector Offset=&AddVectorToPosition(Position=<-30 0 -20>, Vector=$User.offset)
    Offset Amount=1
    Ribbon Width=15
   )
   &AddVelocity(Velocity=<-2000 0 0>) {module=AddVelocity001}
   /SetVariables(Particles.GlowRibbon_Width=50) {isSetParameters=true}
  ]
  particleUpdateScript [
   &FN_DisableTireFX(Mask Index Attribute:NiagaraInt32=$Particles.SpawnGroup, Mask Index FL=2, Mask Index FR=3, Mask Index RL=0, Mask Index RR=1)
   &ParticleState()
   &InheritSourceMovement(
    Applied Position Scale=&Multiply_VectorByFloat(
     Vector=<1 1 1>
     Float=&FloatFromCurve(FloatCurve=DataInterface() {propertyValues {Curve {keys [
          {interpMode=RCIM_Cubic, leaveTangent=-0.0051, tangentMode=RCTM_Break, time=0, value=1}
          {arriveTangent=-0.247, interpMode=RCIM_Cubic, tangentMode=RCTM_Break, time=1, value=0.9}
         ]}}})
    )
    ApplyForce=false
    ApplyPosition=true
    ApplyPositionFalloff=false
   )
   &SolveForcesAndVelocity(Clamp Velocity=false, Limit Acceleration=false)
   &ScaleRibbonWidth(Ribbon Width Scale=&FloatFromCurve(
     FloatCurve=DataInterface() {propertyValues {Curve {keys [
         {interpMode=RCIM_Cubic, leaveTangent=0.1022, tangentMode=RCTM_Break, time=0, value=1}
         {arriveTangent=0.0664, interpMode=RCIM_Cubic, leaveTangent=0.0664, tangentMode=RCTM_User, time=0.3, value=1}
         {arriveTangent=-3.4322, interpMode=RCIM_Cubic, tangentMode=RCTM_Break, time=1, value=0}
        ]}}}
     Scale Curve=1.5
    ))
   &VC_PaintedVariantsOverride(
    Black=<rgba 0.0309 0.0319 0.035 1>
    BurntSienna=<rgba 0.3206 0.0608 0.0037 1>
    Cobalt=&ColorFromCurve(ColorCurve=DataInterface() {propertyValues {
       AlphaCurve {keys [<at 0 1 RCIM_Cubic>,<at 0.9051 0>]}
       BlueCurve {keys [<at 0 1 RCIM_Cubic>,<at 0.0409 1>,<at 0.2227 15>,<at 0.4491 1>,<at 0.5352 3>]}
       GreenCurve {keys [<at 0 0.3612 RCIM_Cubic>,<at 0.0409 0.3612>,<at 0.2227 5.418>,<at 0.4491 0.3612>,<at 0.5352 1.0836>]}
       RedCurve {keys [<at 0 0 RCIM_Cubic>,<at 0.0409 0>,<at 0.2227 0>,<at 0.4491 0>,<at 0.5352 0>]}
      }})
    Crimson=&ColorFromCurve(ColorCurve=DataInterface() {propertyValues {
       AlphaCurve {keys [<at 0 0.2 RCIM_Cubic>,<at 0.3109 0.1>,<at 0.8466 0>]}
       BlueCurve {keys [<at 0 0.0813 RCIM_Cubic>,<at 0.228 0>,<at 0.3354 0.0813>]}
       GreenCurve {keys [<at 0 0 RCIM_Cubic>,<at 0.228 0.3667>,<at 0.3354 0>]}
       RedCurve {keys [<at 0 3 RCIM_Cubic>,<at 0.228 2.2>,<at 0.3354 3>]}
      }})
    DefaultColor=&ColorFromCurve(ColorCurve=DataInterface() {propertyValues {
       AlphaCurve {keys [<at 0 0.2 RCIM_Cubic>,<at 0.1217 0.1>,<at 0.8466 0>]}
       BlueCurve {keys [<at 0 0.8507 RCIM_Cubic>]}
       GreenCurve {keys [<at 0 2.1422 RCIM_Cubic>]}
       RedCurve {keys [<at 0 5 RCIM_Cubic>]}
      }})
    ForestGreen=&ColorFromCurve(ColorCurve=DataInterface() {propertyValues {
       AlphaCurve {keys [<at 0 0.5 RCIM_Cubic>,<at 0.2971 0.15>,<at 0.8466 0>]}
       BlueCurve {keys [<at 0 0.0253 RCIM_Cubic>,<at 0.2418 0.06>,<at 0.3683 0.125>]}
       GreenCurve {keys [<at 0 0.5 RCIM_Cubic>,<at 0.2418 0.6>,<at 0.3683 0.5>]}
       RedCurve {keys [<at 0 0 RCIM_Cubic>,<at 0.2418 0.24>,<at 0.3683 0>]}
      }})
    Grey=<rgba 0.17 0.17 0.17 1>
    Lime=&ColorFromCurve(ColorCurve=DataInterface() {propertyValues {
       AlphaCurve {keys [<at 0 0.7 RCIM_Cubic>,<at 0.2769 0.15>,<at 0.8466 0>]}
       BlueCurve {keys [<at 0 0 RCIM_Cubic>,<at 0.177 0>,<at 0.2843 0>]}
       GreenCurve {keys [<at 0 3 RCIM_Cubic>,<at 0.177 10>,<at 0.2843 5>]}
       RedCurve {keys [<at 0 0.5 RCIM_Cubic>,<at 0.177 1.6667>,<at 0.2843 0.8333>]}
      }})
    Orange=&ColorFromCurve(ColorCurve=DataInterface() {propertyValues {
       AlphaCurve {keys [<at 0 0.7 RCIM_Cubic>,<at 0.2546 0.1>,<at 0.8466 0>]}
       BlueCurve {keys [<at 0 0 RCIM_Cubic>,<at 0.1515 0>,<at 0.2886 0>]}
       GreenCurve {keys [<at 0 1.1667 RCIM_Cubic>,<at 0.1515 2.5>,<at 0.2886 1.1667>]}
       RedCurve {keys [<at 0 7 RCIM_Cubic>,<at 0.1515 15>,<at 0.2886 7>]}
      }})
    Pink=&ColorFromCurve(ColorCurve=DataInterface() {propertyValues {
       AlphaCurve {keys [<at 0 0.6 RCIM_Cubic>,<at 0.1217 0.3>,<at 0.8849 0>]}
       BlueCurve {keys [<at 0.0888 8.6667 RCIM_Cubic>]}
       GreenCurve {keys [<at 0.0888 2 RCIM_Cubic>]}
       RedCurve {keys [<at 0.0888 10 RCIM_Cubic>]}
      }})
    Purple=&ColorFromCurve(ColorCurve=DataInterface() {propertyValues {
       AlphaCurve {keys [<at 0 1 RCIM_Cubic>,<at 0.4916 0.9>,<at 0.8583 0>]}
       BlueCurve {keys [<at 0 1.5 RCIM_Cubic>,<at 0.1685 6>,<at 0.3928 1.5>]}
       GreenCurve {keys [<at 0 0.015 RCIM_Cubic>,<at 0.1685 0.06>,<at 0.3928 0.015>]}
       RedCurve {keys [<at 0 0.2625 RCIM_Cubic>,<at 0.1685 1.05>,<at 0.3928 0.2625>]}
      }})
    R1 Black:LinearColor=<rgba 0 0 0 0.6>
    R1 BurntSienna:LinearColor=<rgba 0.8796 0.1365 0.072 1>
    R1 Cobalt:LinearColor=<rgba 0 0.2665 0.8 1>
    R1 Crimson:LinearColor=<rgba 0.5092 0 0.0085 1>
    R1 DefaultColor=<rgba 0.5583 0.314 0.15 0.8>
    R1 ForestGreen:LinearColor=<rgba 0.0694 0.6 0 0.6>
    R1 Grey:LinearColor=<rgba 0.25 0.25 0.25 1>
    R1 Lime:LinearColor=<rgba 0.2504 0.4 0 1>
    R1 Orange:LinearColor=<rgba 0.5 0.1037 0.0555 1>
    R1 Pink:LinearColor=<rgba 0.6257 0.2051 0.6724 1>
    R1 Purple:LinearColor=<rgba 0.1511 0 0.6724 1>
    R1 Saffron:LinearColor=<rgba 0.4485 0.4544 0.0454 1>
    R1 SkyBlue:LinearColor=<rgba 0 0.2662 0.4 1>
    R1 TitaniumWhite:LinearColor=<rgba 0.4 0.4 0.4 1>
    Saffron=&ColorFromCurve(ColorCurve=DataInterface() {propertyValues {
       AlphaCurve {keys [<at 0 0.7 RCIM_Cubic>,<at 0.8498 0 RCIM_Cubic>]}
       BlueCurve {keys [<at 0 0 RCIM_Cubic>]}
       GreenCurve {keys [<at 0 5 RCIM_Cubic>]}
       RedCurve {keys [<at 0 4.9615 RCIM_Cubic>]}
      }})
    SkyBlue=&ColorFromCurve(ColorCurve=DataInterface() {propertyValues {
       AlphaCurve {keys [<at 0 0.7 RCIM_Cubic>,<at 0.4916 0.35>,<at 0.87 0 RCIM_Cubic>]}
       BlueCurve {keys [<at 0 1.5 RCIM_Cubic>,<at 0.1823 8>,<at 0.4044 1.5>]}
       GreenCurve {keys [<at 0 0.9855 RCIM_Cubic>,<at 0.1823 5.2561>,<at 0.4044 0.9855>]}
       RedCurve {keys [<at 0 0 RCIM_Cubic>,<at 0.1823 0>,<at 0.4044 0>]}
      }})
    TitaniumWhite=&ColorFromCurve(ColorCurve=DataInterface() {propertyValues {
       AlphaCurve {keys [<at 0 1 RCIM_Cubic>,<at 0.8413 0 RCIM_Cubic>]}
       BlueCurve {keys [<at 0 4 RCIM_Cubic>,<at 0.2057 40>,<at 0.4183 4>]}
       GreenCurve {keys [<at 0 4 RCIM_Cubic>,<at 0.2057 40>,<at 0.4183 4>]}
       RedCurve {keys [<at 0 4 RCIM_Cubic>,<at 0.2057 40>,<at 0.4183 4>]}
      }})
    Use R1 Painted Overrides=true
   ) {module=VC_PaintedVariantsOverride001}
   &VC_UpdateRibbon_v1(Enable Skeletal Position=true)
  ]
  rendererClasses [&NiagaraRibbonRendererProperties]
  renderers [
   {rendererClass=&NiagaraRibbonRendererProperties}
   {rendererClass=&NiagaraRibbonRendererProperties, rendererIndex=1}
  ]
 }
 WheelWrap001 {
  emitterUpdateScript [
   &EmitterState(Life Cycle Mode=::NewEnumerator1, Loop Behavior=::NewEnumerator1, Loop Duration Mode=::NewEnumerator1, UseLoopDelay=false)
   &SpawnBurst_Instantaneous(Spawn Count=2, Use Spawn Probability=false, UseLoopCountLimit=false)
  ]
  particleSpawnScript [
   &VC_PaintedVariantsOverride(
    Black=<rgba 0 0 0 50>
    BurntSienna=<rgba 10 2.5425 0 5>
    Cobalt=<rgba 0.25 0.9213 5 3>
    Crimson=<rgba 10 0.6198 0.9183 10>
    DefaultColor=<rgba 50 26.1532 15 1>
    ForestGreen=<rgba 0.006 0.02 0.0067 1>
    Grey=<rgba 0.01 0.01 0.01 6>
    Lime=<rgba 5 10 0 1>
    Orange=<rgba 15 7.6738 3 5>
    Pink=<rgba 70 28 57.1434 3>
    Purple=<rgba 0.0323 0 0.05 3>
    R1 Black:LinearColor=<rgba 0.22 0.22 0.22 0.4>
    R1 BurntSienna:LinearColor=<rgba 0.9 0.225 0 1>
    R1 Cobalt:LinearColor=<rgba 0.8 1.5463 8 0.99>
    R1 Crimson:LinearColor=<rgba 1 0.072 0.04 1>
    R1 DefaultColor=<rgba 5 4.7458 0.625 1>
    R1 ForestGreen:LinearColor=<rgba 0 0.85 0.1183 1>
    R1 Grey:LinearColor=<rgba 0.5642 0.5741 0.62 0.8>
    R1 Lime:LinearColor=<rgba 0.8667 1.3 0.26 1>
    R1 Orange:LinearColor=<rgba 6.6146 1.5474 0 1>
    R1 Pink:LinearColor=<rgba 5 1.5 3.7273 1>
    R1 Purple:LinearColor=<rgba 0.6525 0.1784 0.9911 1>
    R1 Saffron:LinearColor=<rgba 5 4.382 0.2 1>
    R1 SkyBlue:LinearColor=<rgba 0.5082 1.3211 2.1 1>
    R1 TitaniumWhite:LinearColor=<rgba 5 5 5 1>
    Saffron=<rgba 9 8.073 1.8 3>
    SkyBlue=<rgba 1.7865 5.2555 7 1>
    TitaniumWhite=<rgba 25 25 25 3>
    Use R1 Painted Overrides=true
   )
   &InitializeParticle(
    Color Mode=::NewEnumerator0
    Lifetime=$Emitter.CurrentLoopDuration
    Mesh Scale Mode=::NewEnumerator4
    Sprite Rotation Mode=::NewEnumerator2
    Sprite Size Mode=::NewEnumerator3
    UsePositionOffset=false
    Write Material Random=true
   )
   &PartitionParticles()
   &SkeletalMeshLocation(
    Bone / Socket Index=&Add_Integer(A=$Particles.PartitionParticles.Partition, B=4)
    Bone / Socket Sampling Mode=::NewEnumerator5
    Interpolate Mesh Sampling=false
    MaskSpawnGroup=false
    Mesh Sampling Type=::NewEnumerator6
    Skeletal Mesh=$System.SkeletalMesh
   )
  ]
  particleUpdateScript [
   &FN_DisableTireFX(
    Mask Index Attribute:NiagaraInt32=$Particles.PartitionParticles.Partition
    Mask Index FL=2
    Mask Index FR=3
    Mask Index RL=0
    Mask Index RR=1
   )
   &VC_WheelWrapScale(Wheel Index Attribute:NiagaraInt32=&Add_Integer(A=$Particles.PartitionParticles.Partition, B=2))
   &ParticleState()
   &UpdateMeshOrientation(Rotation Coordinate Space=::NewEnumerator2, Rotation Vector=<0 1 0>)
   &SkeletalMeshLocation(
    Bone / Socket Index=&Add_Integer(A=$Particles.PartitionParticles.Partition, B=4)
    Bone / Socket Sampling Mode=::NewEnumerator5
    MaskSpawnGroup=false
    Mesh Sampling Type=::NewEnumerator6
    Skeletal Mesh=$System.SkeletalMesh
   ) {module=SkeletalMeshLocation001}
   &SolveForcesAndVelocity(Clamp Velocity=false, Limit Acceleration=false)
   &ScaleColor(Scale RGB=&Multiply_VectorByFloat(Vector=<1 1 1>, Float=500), ScaleA=true, ScaleRGB=true)
  ]
  rendererClasses [&NiagaraMeshRendererProperties]
  renderers [
   {rendererClass=&NiagaraMeshRendererProperties, rendererIndex=0}
   {rendererClass=&NiagaraMeshRendererProperties, rendererIndex=1}
  ]
 }
}
summary {
 systemName=NS_DriftTrail_GroundCracks_MP
 emitters [
  {emitterName=Sparks001, simTarget=GPUComputeSim, rendererClasses [&NiagaraSpriteRendererProperties]}
  {emitterName=Trail001, rendererClasses [&NiagaraRibbonRendererProperties]}
  {emitterName=WheelWrap001, rendererClasses [&NiagaraMeshRendererProperties]}
  {emitterName=Glow_Fake001, rendererClasses [&NiagaraSpriteRendererProperties]}
  {
   emitterName=Rocks_V001
   simTarget=GPUComputeSim
   rendererClasses [&NiagaraMeshRendererProperties, &NiagaraSpriteRendererProperties]
  }
 ]
 userVariables [
  User.bIsFrontendPreview {type=NiagaraBool}
  User.bIsPainted {type=NiagaraBool}
  User.ColorVariantID {type=NiagaraInt32}
  User.offset {type=Vector3f, description="offset from wheels(set through cosmetic template)", defaultValue {struct=Vector3f, value=<0 0 0>}}
  User.PaintedColor {type=LinearColor, description="", defaultValue {struct=LinearColor, value=<rgba 1 1 1 1>}}
  User.PaintedColor_1 {type=LinearColor, description="", defaultValue {struct=LinearColor, value=<rgba 1 1 1 1>}}
  User.PaintedColor_2 {type=LinearColor, description="", defaultValue {struct=LinearColor, value=<rgba 1 1 1 1>}}
  User.PaintedColor_3 {type=LinearColor, description="", defaultValue {struct=LinearColor, value=<rgba 1 1 1 1>}}
  User.PaintedColor_4 {type=LinearColor, description="", defaultValue {struct=LinearColor, value=<rgba 1 1 1 1>}}
  User.PaintedColor_5 {type=LinearColor, description="", defaultValue {struct=LinearColor, value=<rgba 1 1 1 1>}}
  User.PaintedColor_4v2 {type=LinearColor, description="", defaultValue {struct=LinearColor, value=<rgba 1 1 1 1>}}
  User.PaintedScalar {type=NiagaraFloat, defaultValue=1, description=""}
  User.VehicleFlags {type=NiagaraInt32, defaultValue=806289408}
  User.Wheel_Radius_Array {
   type=NiagaraDataInterfaceArrayFloat
   defaultValue {
    struct=NiagaraExt_VariableValue_DataInterface
    value {
     dataInterface=&self:NiagaraDataInterfaceArrayFloat_1
     dataInterfaceClass=&NiagaraDataInterfaceArrayFloat
    }
   }
  }
  User.Wheel_Width_Array {
   type=NiagaraDataInterfaceArrayFloat
   defaultValue {
    struct=NiagaraExt_VariableValue_DataInterface
    value {
     dataInterface=&self:NiagaraDataInterfaceArrayFloat_0
     dataInterfaceClass=&NiagaraDataInterfaceArrayFloat
    }
   }
  }
  User.WheelTouchingArray {
   type=NiagaraDataInterfaceArrayFloat
   defaultValue {
    struct=NiagaraExt_VariableValue_DataInterface
    value {
     dataInterface=&self:NiagaraDataInterfaceArrayFloat_2
     dataInterfaceClass=&NiagaraDataInterfaceArrayFloat
    }
   }
  }
 ]
}
